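# Looping exterior ambient spot cue of a wind-excited chain-link fence rattle, filed under Mountain Rescue Station.
# One looping wave runs through two nested, endlessly looping 40-second envelopes: the inner modulates pitch, the outer makes volume dip and swell.
# A quality node feeds that chain to its first two levels and leaves the other two empty.
# Playback is quiet, uses chain-link fence attenuation, caps concurrency and sends to the exterior ambient bus.
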SoundCue /Game/Sounds/Helios/POI/MountainRescueStation/Spots/Spot_Wind_Metal_Excited_Chain_Link_Fence_Loop_Ext_Cue
root=QualityLevel_0
nodes [
 {
  id=Enveloper_0
  children [WavePlayer_0]
  properties {
   bLoop=true
   bLoopIndefinitely=true
   loopEnd=40
   pitchCurve {editorCurveData {keys [
      {interpMode=RCIM_Cubic, time=0}
      <at 10 0.9 RCIM_Cubic>
      {arriveTangent=-0.0001, interpMode=RCIM_Cubic, leaveTangent=-0.0001, tangentMode=RCTM_User, time=20}
      <at 30 0.9 RCIM_Cubic>
      {interpMode=RCIM_Cubic, time=40}
     ]}}
   volumeCurve {editorCurveData {keys [<at 0 1>]}}
  }
 }
 {
  id=Enveloper_3
  children [Enveloper_0]
  properties {
   bLoop=true
   bLoopIndefinitely=true
   loopEnd=40
   pitchCurve {editorCurveData {keys [{time=0}]}}
   volumeCurve {editorCurveData {keys [
      {arriveTangent=-0.01, interpMode=RCIM_Cubic, time=0, value=1}
      {arriveTangent=-0.0428, interpMode=RCIM_Cubic, leaveTangent=-0.0428, time=9.9314, value=0.1}
      {arriveTangent=0.1303, interpMode=RCIM_Cubic, leaveTangent=0.1303, tangentMode=RCTM_User, time=17.34, value=0.2583}
      {arriveTangent=-0.0126, interpMode=RCIM_Cubic, leaveTangent=-0.0126, time=20, value=1}
      {arriveTangent=-0.0465, interpMode=RCIM_Cubic, leaveTangent=-0.0465, time=29.9314, value=0.1}
      {arriveTangent=0.1852, interpMode=RCIM_Cubic, leaveTangent=0.1852, tangentMode=RCTM_User, time=36.7959, value=0.2187}
      <at 40 1>
     ]}}
  }
 }
 {id=QualityLevel_0, children [Enveloper_3,Enveloper_3,null,null]}
 {
  id=WavePlayer_0
  properties {bLooping=true, soundWaveAssetPtr=/Game/Sounds/Artemis/POI/Hollow_Base/Spots/Spot_Wind_Excited_Chainlink_Fence_Rattle_01}
 }
]
properties {
 attenuationSettings=&/Spot_Metal_ChainLink_Fence_Small_Att
 bOverrideConcurrency=true
 maxDistance=978.0443
 soundClassObject=&Fort_Ambient_Outdoors_Athena
 volumeMultiplier=0.1
 attenuationOverrides {falloffDistance=1600}
 busSends [{audioBus=null, minSendLevel=1, sendLevel=1, soundSourceBus=&AmbientAudio_Exterior_SourceBus}]
 concurrencyOverrides {maxCount=2, voiceStealReleaseTime=2}
}
